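# Sound cue for the Butterfly event's "Brk" sound, played at map-wide range through a map-wide events sound class and attenuation.
# A quality-level node routes two slots to a Stage switch and the other two to a single optimized first-stage wave.
# The switch selects one of three Brk waves by the Stage integer parameter, with the first also as default and the third boosted in volume.
# A ButterflySpin concurrency set is assigned.
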SoundCue /Figment_S06/Athena/Sounds/Events/Butterfly/Butterfly_Brk_Cue
root=QualityLevel_0
nodes [
 {id=Modulator_0, children [WavePlayer_4], properties {volumeMax=1.3, volumeMin=1.3}}
 {id=QualityLevel_0, children [Switch_0,Switch_0,WavePlayer_2,WavePlayer_2]}
 {id=Switch_0, children [WavePlayer_0,WavePlayer_0,WavePlayer_1,Modulator_0], properties {intParameterName=Stage}}
 {id=WavePlayer_0, properties {soundWaveAssetPtr=&/Butterfly_Brk1_01}}
 {id=WavePlayer_1, properties {soundWaveAssetPtr=&/Butterfly_Brk2_01}}
 {id=WavePlayer_2, properties {soundWaveAssetPtr=&/Butterfly_Brk1_01_opt}}
 {id=WavePlayer_4, properties {soundWaveAssetPtr=&/Butterfly_Brk3_01}}
]
properties {
 attenuationSettings=/Figment_S06/Sounds/Attenuation/Fort_Butterfly_Mapwide_Att
 maxDistance=2097152
 soundClassObject=/Figment_S06/Athena/Sounds/Events/SoundClasses/Fort_Events_Mapwide
 volumeMultiplier=1.5
 concurrencyOverrides {maxCount=6, resolutionRule=StopOldest}
 concurrencySet [/Figment_S06/Sounds/Concurrency_Rules/Athena_ButterflySpin_Conc]
}
