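# Sound cue for a "Drp" sound in the Figment S06 Butterfly event, set up for map-wide playback but with no root or sound nodes in its graph.
# Its settings pair a map-wide Butterfly attenuation and very large max distance with a map-wide events sound class and boosted volume.
# It references the ButterflySpin concurrency set and also carries override values that cap voices, stopping the oldest.
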SoundCue /Figment_S06/Athena/Sounds/Events/Butterfly/Butterfly_Drp_Drp_Cue
root=null
nodes []
properties {
 attenuationSettings=/Figment_S06/Sounds/Attenuation/Fort_Butterfly_Mapwide_Att
 maxDistance=1000000
 soundClassObject=/Figment_S06/Athena/Sounds/Events/SoundClasses/Fort_Events_Mapwide
 volumeMultiplier=1.5
 concurrencyOverrides {maxCount=6, resolutionRule=StopOldest}
 concurrencySet [/Figment_S06/Sounds/Concurrency_Rules/Athena_ButterflySpin_Conc]
}
