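# Sound cue for the start of the Butterfly event drop sound in Athena.
# A random node picks between two drop wave variations with equal weight.
# It plays at boosted volume through a mapwide attenuation and the mapwide events sound class, with a very large audible range.
# It references a ButterflySpin concurrency set and also carries override settings that cap instances and stop the oldest.
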SoundCue /Figment_S06/Athena/Sounds/Events/Butterfly/Butterfly_Drp_Strt_Cue
root=Random_0
nodes [
 {id=Random_0, children [WavePlayer_1,WavePlayer_2], properties {weights [1,1]}}
 {id=WavePlayer_1, properties {soundWaveAssetPtr=&/Butterfly_Drp_02}}
 {id=WavePlayer_2, properties {soundWaveAssetPtr=&/Butterfly_Drp_01}}
]
properties {
 attenuationSettings=/Figment_S06/Sounds/Attenuation/Fort_Butterfly_Mapwide_Att
 maxDistance=2097152
 soundClassObject=/Figment_S06/Athena/Sounds/Events/SoundClasses/Fort_Events_Mapwide
 volumeMultiplier=1.5
 concurrencyOverrides {maxCount=6, resolutionRule=StopOldest}
 concurrencySet [/Figment_S06/Sounds/Concurrency_Rules/Athena_ButterflySpin_Conc]
}
