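# Sound cue for the spin sound of the Butterfly event in Figment_S06, routed through a map-wide sound class and attenuation.
# A quality-level node at the root scales variety: the first level picks randomly from four waves, the second from two, and the remaining two share a single "opt" wave.
# A shared Butterfly spin concurrency asset governs simultaneous playback.
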SoundCue /Figment_S06/Athena/Sounds/Events/Butterfly/Butterfly_Spin_Cue
root=QualityLevel_0
nodes [
 {id=QualityLevel_0, children [Random_0,Random_1,WavePlayer_4,WavePlayer_4]}
 {id=Random_0, children [WavePlayer_2,WavePlayer_3,WavePlayer_6,WavePlayer_7], properties {weights [1,1,1,1]}}
 {id=Random_1, children [WavePlayer_0,WavePlayer_1], properties {weights [1,1]}}
 {id=WavePlayer_0, properties {soundWaveAssetPtr=&/Butterfly_Spoooon_04}}
 {id=WavePlayer_1, properties {soundWaveAssetPtr=&/Butterfly_Spoon_01}}
 {id=WavePlayer_2, properties {soundWaveAssetPtr=&Butterfly_Spoooon_04}}
 {id=WavePlayer_3, properties {soundWaveAssetPtr=&Butterfly_Spoon_01}}
 {id=WavePlayer_4, properties {soundWaveAssetPtr=&/Butterfly_Spoon_01_opt}}
 {id=WavePlayer_6, properties {soundWaveAssetPtr=&/Butterfly_Spooon_02}}
 {id=WavePlayer_7, properties {soundWaveAssetPtr=&/Butterfly_Spooon_03}}
]
properties {
 attenuationSettings=/Figment_S06/Sounds/Attenuation/Fort_Butterfly_Mapwide_Att
 maxDistance=2097152
 soundClassObject=/Figment_S06/Athena/Sounds/Events/SoundClasses/Fort_Events_Mapwide
 volumeMultiplier=1.75
 concurrencyOverrides {maxCount=6, resolutionRule=StopOldest}
 concurrencySet [/Figment_S06/Sounds/Concurrency_Rules/Athena_ButterflySpin_Conc]
}
