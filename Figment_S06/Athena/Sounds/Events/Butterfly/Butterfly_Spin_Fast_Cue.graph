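# Sound cue for a fast butterfly spin in the Figment_S06 Athena events, using a mapwide attenuation and event sound class.
# A quality-level root selects an equally weighted random pool of four SpoonMed wave variants for the first level or two for the second, leaving the other slots empty.
# The cue plays pitched up, with its own concurrency limit that stops the oldest instance.
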SoundCue /Figment_S06/Athena/Sounds/Events/Butterfly/Butterfly_Spin_Fast_Cue
root=QualityLevel_0
nodes [
 {id=QualityLevel_0, children [Random_1,Random_0,null,null]}
 {id=Random_0, children [WavePlayer_2,WavePlayer_3], properties {weights [1,1]}}
 {id=Random_1, children [WavePlayer_0,WavePlayer_1,WavePlayer_4,WavePlayer_5], properties {weights [1,1,1,1]}}
 {id=WavePlayer_0, properties {soundWaveAssetPtr=&/Butterfly_SpoonMed_01}}
 {id=WavePlayer_1, properties {soundWaveAssetPtr=&/Butterfly_SpooonMed_02}}
 {id=WavePlayer_2, properties {soundWaveAssetPtr=&Butterfly_SpoonMed_01}}
 {id=WavePlayer_3, properties {soundWaveAssetPtr=&Butterfly_SpooonMed_02}}
 {id=WavePlayer_4, properties {soundWaveAssetPtr=&/Butterfly_SpoooonMed_03}}
 {id=WavePlayer_5, properties {soundWaveAssetPtr=&/Butterfly_SpooonMed_04}}
]
properties {
 attenuationSettings=/Figment_S06/Sounds/Attenuation/Fort_Butterfly_Mapwide_Att
 bOverrideConcurrency=true
 maxDistance=2097152
 pitchMultiplier=1.5
 soundClassObject=/Figment_S06/Athena/Sounds/Events/SoundClasses/Fort_Events_Mapwide
 volumeMultiplier=1
 concurrencyOverrides {maxCount=6, resolutionRule=StopOldest}
}
