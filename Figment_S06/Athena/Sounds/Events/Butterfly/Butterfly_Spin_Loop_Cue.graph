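# Sound cue for the spin loop of the Butterfly event, played with map-wide attenuation and an ambient-ducking sound class.
# The SpinRamp parameter drives pitch and volume of two build-up loops and a lightning beam loop over staggered ranges.
# A mixer adds randomly delayed, repeating Cube zap vocals and looping random distant thunder that fades out as SpinRamp rises.
# Quality-level nodes swap in an optimized wave, trim thunder variations or drop layers.
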SoundCue /Figment_S06/Athena/Sounds/Events/Butterfly/Butterfly_Spin_Loop_Cue
root=Mixer_0
nodes [
 {id=Delay_0, children [QualityLevel_2], properties {delayMax=3, delayMin=1}}
 {id=Looping_0, children [Delay_0]}
 {id=Looping_1, children [QualityLevel_6]}
 {
  id=Mixer_0
  children [ModulatorContinuous_2, ModulatorContinuous_1, ModulatorContinuous_0, QualityLevel_0, ParamCrossFade_0]
  properties {inputVolume [1.5,1,1,1,0.3]}
 }
 {
  id=ModulatorContinuous_0
  children [WavePlayer_0]
  properties {
   pitchModulationParams {default=0, maxOutput=1.5, minInput=0.3, minOutput=0.5, parameterName=SpinRamp}
   volumeModulationParams {default=0, maxOutput=1.5, minInput=0.5, parameterName=SpinRamp}
  }
 }
 {
  id=ModulatorContinuous_1
  children [QualityLevel_3]
  properties {
   pitchModulationParams {default=0, maxOutput=1.5, minOutput=0.5, parameterName=SpinRamp}
   volumeModulationParams {default=0, maxInput=0.58, maxOutput=1.5, minInput=0.1, parameterName=SpinRamp}
  }
 }
 {
  id=ModulatorContinuous_2
  children [QualityLevel_1]
  properties {
   pitchModulationParams {default=0, maxInput=0.3, maxOutput=0.8, minInput=0.05, minOutput=0.5, parameterName=SpinRamp}
   volumeModulationParams {default=0, maxInput=0.1, maxOutput=1.5, minInput=0.01, parameterName=SpinRamp}
  }
 }
 {
  id=ParamCrossFade_0
  children [Looping_1]
  properties {
   paramName=SpinRamp
   crossFadeInput [{fadeInDistanceEnd=0, fadeInDistanceStart=0, fadeOutDistanceEnd=0.5, fadeOutDistanceStart=0.3, volume=1}]
  }
 }
 {id=QualityLevel_0, children [Looping_0,Looping_0,null,null]}
 {id=QualityLevel_1, children [WavePlayer_4,WavePlayer_4,null,null]}
 {id=QualityLevel_2, children [Random_2,Random_2,null,null]}
 {id=QualityLevel_3, children [WavePlayer_1,WavePlayer_1,WavePlayer_2,WavePlayer_2]}
 {id=QualityLevel_6, children [Random_0,Random_3,null,null]}
 {
  id=Random_0
  children [WavePlayer_11, WavePlayer_12, WavePlayer_13, WavePlayer_14, WavePlayer_15, WavePlayer_16, WavePlayer_17, WavePlayer_18, WavePlayer_19]
  properties {weights [1,1,1,1,1,1,1,1,1]}
 }
 {id=Random_2, children [WavePlayer_5,WavePlayer_8], properties {weights [1,1]}}
 {id=Random_3, children [WavePlayer_20, WavePlayer_21, WavePlayer_22], properties {weights [1,1,1]}}
 {id=WavePlayer_0, properties {bLooping=true, soundWaveAssetPtr=/Game/Athena/Sounds/Events/CrackLightning/Crack_Lightning_BeamLoop_01}}
 {id=WavePlayer_1, properties {bLooping=true, soundWaveAssetPtr=/Game/Athena/Sounds/Events/Butterfly/Butterfly_BuildUp_2}}
 {id=WavePlayer_11, properties {soundWaveAssetPtr=/Game/Sounds/Fort_GamePlay_Sounds/Storms/Dist_Thunder_001}}
 {id=WavePlayer_12, properties {soundWaveAssetPtr=/Game/Sounds/Fort_GamePlay_Sounds/Storms/Dist_Thunder_002}}
 {id=WavePlayer_13, properties {soundWaveAssetPtr=/Game/Sounds/Fort_GamePlay_Sounds/Storms/Dist_Thunder_003}}
 {id=WavePlayer_14, properties {soundWaveAssetPtr=/Game/Sounds/Fort_GamePlay_Sounds/Storms/Dist_Thunder_004}}
 {id=WavePlayer_15, properties {soundWaveAssetPtr=/Game/Sounds/Fort_GamePlay_Sounds/Storms/Dist_Thunder_005}}
 {id=WavePlayer_16, properties {soundWaveAssetPtr=/Game/Sounds/Fort_GamePlay_Sounds/Storms/Dist_Thunder_006}}
 {id=WavePlayer_17, properties {soundWaveAssetPtr=/Game/Sounds/Fort_GamePlay_Sounds/Storms/Dist_Thunder_007}}
 {id=WavePlayer_18, properties {soundWaveAssetPtr=/Game/Sounds/Fort_GamePlay_Sounds/Storms/Dist_Thunder_008}}
 {id=WavePlayer_19, properties {soundWaveAssetPtr=/Game/Sounds/Fort_GamePlay_Sounds/Storms/Dist_Thunder_009}}
 {id=WavePlayer_2, properties {bLooping=true, soundWaveAssetPtr=&/Butterfly_BuildUp_2_opt}}
 {id=WavePlayer_20, properties {soundWaveAssetPtr=&Dist_Thunder_001}}
 {id=WavePlayer_21, properties {soundWaveAssetPtr=&Dist_Thunder_002}}
 {id=WavePlayer_22, properties {soundWaveAssetPtr=&Dist_Thunder_003}}
 {id=WavePlayer_4, properties {bLooping=true, soundWaveAssetPtr=/Game/Athena/Sounds/Events/Butterfly/Butterfly_BuildUp_1}}
 {id=WavePlayer_5, properties {soundWaveAssetPtr=/Game/Athena/Sounds/Events/CubeIsland/Cube_SpawnerZapVocal_01}}
 {id=WavePlayer_8, properties {soundWaveAssetPtr=/Game/Athena/Sounds/Events/CubeIsland/Cube_SpawnerZapVocal_02}}
]
properties {
 attenuationSettings=/Figment_S06/Sounds/Attenuation/Fort_Butterfly_Mapwide_Att
 bOverrideConcurrency=true
 maxDistance=2097152
 soundClassObject=/Figment_S06/Sounds/FortSoundClassesAndModes/Fort_Nexus_DuckAmbient
 volumeMultiplier=1
 concurrencyOverrides {maxCount=6, resolutionRule=StopOldest}
}
