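# Sound cue for the medium spin sound of the Butterfly event, set up with map-wide attenuation and sound class.
# A quality-level node picks an equally weighted random set of four waves at the first level or two at the second, leaving the other two levels empty.
# The cue plays pitched down and slightly boosted, under the ButterflySpin concurrency rule.
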SoundCue /Figment_S06/Athena/Sounds/Events/Butterfly/Butterfly_Spin_Med_Cue
root=QualityLevel_0
nodes [
 {id=QualityLevel_0, children [Random_0,Random_2,null,null]}
 {id=Random_0, children [WavePlayer_2,WavePlayer_3,WavePlayer_6,WavePlayer_7], properties {weights [1,1,1,1]}}
 {id=Random_2, children [WavePlayer_8,WavePlayer_9], properties {weights [1,1]}}
 {id=WavePlayer_2, properties {soundWaveAssetPtr=&/Butterfly_SpoonMed_01}}
 {id=WavePlayer_3, properties {soundWaveAssetPtr=&/Butterfly_SpooonMed_02}}
 {id=WavePlayer_6, properties {soundWaveAssetPtr=&/Butterfly_SpoooonMed_03}}
 {id=WavePlayer_7, properties {soundWaveAssetPtr=&/Butterfly_SpooonMed_04}}
 {id=WavePlayer_8, properties {soundWaveAssetPtr=&Butterfly_SpoonMed_01}}
 {id=WavePlayer_9, properties {soundWaveAssetPtr=&Butterfly_SpooonMed_02}}
]
properties {
 attenuationSettings=/Figment_S06/Sounds/Attenuation/Fort_Butterfly_Mapwide_Att
 maxDistance=2097152
 pitchMultiplier=0.75
 soundClassObject=/Figment_S06/Athena/Sounds/Events/SoundClasses/Fort_Events_Mapwide
 volumeMultiplier=1.2
 concurrencyOverrides {maxCount=6, resolutionRule=StopOldest}
 concurrencySet [/Figment_S06/Sounds/Concurrency_Rules/Athena_ButterflySpin_Conc]
}
